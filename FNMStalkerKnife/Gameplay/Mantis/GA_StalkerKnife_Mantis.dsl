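; Mantis melee ability for the Stalker Knife, layering combo, lunge and air-attack behavior onto technique events.
; Technique starts choose the targeting preset, apply lunge or air-attack effects, and drive camera modifiers and air-attack energy cost.
; Landing, surface swimming or slope sliding jumps the air attack to its montage Land section with camera and force feedback.
; Hits spend weapon durability; hits, death and technique end remove the granted effect.
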(event Ability|EventActivateAbility
  ; Bind to OnLanded to jump to Land section of montage
  (|CallParentFunction)
  (Development|PrintString "Matins is active")
  (bind _outfortplayerpawn (CallFunction|GetAvatarasValidFortPlayerPawn)
    (:ValidFortPlayerPawn
      (bind _output_get (Variables|Default|SetPlayer%20Pawn _outfortplayerpawn))
      (bind _asb_stalker_knife_mantis_weapon (Utilities|Casting|CastToB_StalkerKnife_MantisWeapon (Weapon|GetCurrentWeapon _output_get)))
      (Variables|Default|SetCurrentWeapon _asb_stalker_knife_mantis_weapon)
      (EventDispatchers|BindEventtoOnPawnLanded _output_get (EventDispatchers|CreateEvent :!event Custom|PawnLanded))
      (EventDispatchers|BindEventtoOnDeathEffects _output_get (EventDispatchers|CreateEvent :!event Custom|PlayerDeathEffects))
      (bind _asfort_movement_comp_character (Utilities|Casting|CastToFortMovementComp_Character (Pawn|GetMovementComponent _output_get)))
      (bind _output_get_2 (Variables|Default|SetFort%20Movement%20Comp%20Character _asfort_movement_comp_character))
      (Class|FortMovementCompCharacter|CancelActiveExtensionwithTag _output_get_2 (Variables|Default|GetCustomTacSprintTag))
      (CallFunction|SetEnableCameraModifier true)
      (bind _tacticalsprintenergycomponent (Default|GetTacticalSprintEnergyComponent)
        (:IsValid
          (Variables|Default|SetTactical%20Sprint%20Energy%20Component _tacticalsprintenergycomponent)
          (Ability|Tasks|WaitGameplayTagAddToActor _output_get "(TagName=\"Gameplay.Action.Player.SurfaceSwimming\")"
            (:then
              (Ability|Tasks|WaitGameplayTagAddToActor _output_get "(TagName=\"Movement.Status.SlopeSliding\")"
                (:Added
                  (if (not (Variables|Default|GetbLanded))
                    (CallFunction|PawnLanded)))))
            (:Added
              (CallFunction|PawnLanded))))))))

(event Ability|EventOnEndAbility
  (params (bWasCancelled :type Boolean))
  (|CallParentFunction bWasCancelled))

(event Mantis|EventOnMantisAblilityTechniqueStarted
  (params (OwningTags :type Gameplay%20Tag%20Container%20Structure) (SequenceName :type Name :default "None") (StartsSequence :type Boolean) (EndsSequence :type Boolean))
  ; Comment
  (|CallParentFunction OwningTags SequenceName StartsSequence EndsSequence)
  (Variables|Default|SetTargeting%20Preset (select (== SequenceName "Lunge") (Variables|Default|GetTP_Lunge) (Variables|Default|GetTP_Attack)))
  (switch name SequenceName
    (:Combo1
      _)
    (:Combo2
      _)
    (:Combo3
      _)
    (:Lunge
      (bind _returnvalue (Ability|ApplyGameplayEffectToOwner "/FNMStalkerKnife/Gameplay/GE_StalkerKnife_GrantLunge.GE_StalkerKnife_GrantLunge_C"))
      (Variables|Default|SetGELungeHandle _returnvalue)
      (CallFunction|PlayLungeLandingCameraAnim))
    (:AirAttack01
      (Variables|Default|SetbCanLandAirAttack true)
      (bind _returnvalue_2 (Ability|ApplyGameplayEffectToOwner "/FNMStalkerKnife/Gameplay/GE_StalkerKnife_GrantAirAttack.GE_StalkerKnife_GrantAirAttack_C"))
      (Variables|Default|SetGELungeHandle _returnvalue_2)
      (bind _returnvalue_3 (Fortnite|Camera|AddCameraModifierforOwningPlayerController (Variables|Default|GetDiveCameraModifierClass) (Variables|Default|GetPlayer%20Pawn)))
      (Variables|Default|SetAir%20Attack%20Camera%20Modifier%20Handle (Utilities|Casting|CastToFortCameraModifier_TargetRotationFrustum _returnvalue_3))
      (bind _tactical_sprint_energy_component (Variables|Default|GetTactical%20Sprint%20Energy%20Component)
        (:then
          (bind (_outamountofenergyused _returnvalue_4) (Energy|UseEnergy _tactical_sprint_energy_component (ScalableFloat|GetValueatLevel (Variables|Default|GetAirAttackEnergyCost)) true 0))
          (if (not _returnvalue_4)
            (bind _returnvalue_5 (Ability|ApplyGameplayEffectToOwner "/FNMStalkerKnife/Gameplay/GE_StalkerKnife_AirAttackEnergyTag.GE_StalkerKnife_AirAttackEnergyTag_C"))
            (Variables|Default|SetOutOfEnergyForAirAttackHandle _returnvalue_5)))
        (:else))))
  (CallFunction|PlayForceFeedback 0))

(event Mantis|EventOnMantisAblilityTechniqueEnded
  (params (SequenceName :type Name :default "None"))
  ; Comment
  (|CallParentFunction SequenceName)
  (switch name SequenceName
    (:Lunge
      (CallFunction|PlayForceFeedback (Variables|Default|GetForceFeedbackHeavy)))
    (:AirAttack01
      (Variables|Default|SetbCanLandAirAttack false)
      (Variables|Default|SetbLanded false)))
  (Ability|RemoveGameplayEffectFromOwnerWithHandle (Variables|Default|GetGELungeHandle))
  (CallFunction|SetEnableCameraModifier false)
  (bind _air_attack_camera_modifier_handle (Variables|Default|GetAir%20Attack%20Camera%20Modifier%20Handle)
    (:then
      (Class|FortCameraModifierTargetRotationFrustum|RemoveSelffromCameraManager _air_attack_camera_modifier_handle))
    (:else)))

(event Custom|PawnLanded
  (params (Hit :type Hit%20Result%20Structure%20%28by%20ref%29))
  ; Jump to Land Section to finish the AirAttack
  (if (and (Variables|Default|GetbCanLandAirAttack) (not (Variables|Default|GetbLanded)))
    (Variables|Default|SetbLanded true)
    (Ability|Animation|MontageJumptoSection "Land")
    (Ability|RemoveGameplayEffectFromOwnerWithHandle (Variables|Default|GetGELungeHandle))
    (CallFunction|PlayLungeLandingCameraAnim)
    (CallFunction|PlayForceFeedback (Variables|Default|GetForceFeedbackHeavy))))

(event Mantis|EventOnMantisTechniqueHit
  (params (TargetDataHandle :type Gameplay%20Ability%20Target%20Data%20Handle%20Structure%20%28by%20ref%29) (ApplicationTag :type Gameplay%20Tag%20Structure))
  ; Comment
  (bind _output_get (Variables|Default|SetTarget%20Data%20Handle TargetDataHandle))
  (|CallParentFunction _output_get ApplicationTag)
  (Ability|RemoveGameplayEffectFromOwnerWithHandle (Variables|Default|GetGELungeHandle))
  (bind _currentweapon (Variables|Default|GetCurrentWeapon)
    (:then
      (FortWeapon|UseWeaponDurability(Explicit) _currentweapon (ScalableFloat|GetValueatLevel (Variables|Hotfix|GetHotfix_HitDurabilityCost))))
    (:else)))

(event Custom|PlayerDeathEffects
  (params (DeadActor :type Actor%20Object%20Reference) (Damage :type Float%20%28single-precision%29) (InstigatedBy :type Fort%20Pawn%20Object%20Reference) (HitLocation :type Vector) (DamageCauser :type Actor%20Object%20Reference))
  (Ability|RemoveGameplayEffectFromOwnerWithHandle (Variables|Default|GetGELungeHandle)))

(fn PlayForceFeedback
  (params (ForceFeedbackEffect :type Force%20Feedback%20Effect%20Object%20Reference))
  (bind _returnvalue (CallFunction|GetAvatarasFortPlayerPawn))
  (bind _returnvalue_2 (Fortnite|GetFortPlayerControllerfromActor _returnvalue))
  ;[0.008391 0.1308 0.1406] Play force feedback on the avatar player pawn.
  (Utilities|IsValid _returnvalue
    (:Is%20Valid
      (Utilities|IsValid _returnvalue_2
        (:Is%20Valid
          (Game|Feedback|LocalPlayForceFeedback _returnvalue_2 ForceFeedbackEffect false "None"))))))

(fn PlayLungeLandingCameraAnim
  (bind _player_pawn (Variables|Default|GetPlayer%20Pawn))
  (bind _returnvalue (Fortnite|GetFortPlayerControllerfromActor _player_pawn))
  (bind _lungelandingcameraanimation (Variables|Default|GetLungeLandingCameraAnimation)
    (:then
      (bind _player_pawn_2 (Variables|Default|GetPlayer%20Pawn)
        (:then
          (bind _playercameramanager (Variables|/Script/Engine.PlayerController|GetPlayerCameraManager (Fortnite|GetFortPlayerControllerfromActor _player_pawn_2))
            (:then
              (Utilities|IsValid (CameraAnimation|ConvCameraAnimationCameraModifier _playercameramanager)
                (:Is%20Valid
                  (bind _returnvalue_2 (CameraAnimation|PlayCameraAnimation (CameraAnimation|ConvCameraAnimationCameraModifier _playercameramanager) _lungelandingcameraanimation (Utilities|Struct|MakeCameraAnimationParams 1.0 1.0 "Linear" 0.2 "Linear" 1.0) :!class /Script/EngineCameras.CameraAnimationCameraModifier))
                  (Variables|Default|SetLungeLandingCameraAnimationHandle _returnvalue_2)
                  (Ability|HasAuthority
                    (:False
                      (Utilities|IsValid _returnvalue
                        (:Is%20Valid
                          (Game|Feedback|ClientStartCameraShake :self _returnvalue :Shake (Variables|Default|GetLungeLandingCameraShake) :Scale 1.0 :PlaySpace "CameraLocal"))))))))
            (:else)))
        (:else)))
    (:else)))

(fn SetEnableCameraModifier
  (params (Enable :type Boolean))
  (bind _mantiscameramodifierclass (Variables|Default|GetMantisCameraModifierClass))
  ;[0.000999 0.15 0.06155 0.5] SetEnableMantisCameraModifier: Add the camera modifier for that will be used to adjust the camera during the melee attacks.
  (bind _player_pawn (Variables|Default|GetPlayer%20Pawn)
    (:then
      (Utilities|IsValid (Fortnite|GetFortPlayerControllerfromActor _player_pawn)
        (:Is%20Valid
          (if Enable
            (bind _playercameramanager (Variables|/Script/Engine.PlayerController|GetPlayerCameraManager (Fortnite|GetFortPlayerControllerfromActor _player_pawn))
              (:then
                (bind _returnvalue (Game|Player|FindCameraModifierbyClass _playercameramanager _mantiscameramodifierclass))
                (Utilities|IsValid _returnvalue
                  (:Is%20Valid
                    (bind _ascamera_modifier_stalker_knife_mantis (Utilities|Casting|CastToCameraModifier_StalkerKnife_Mantis _returnvalue))
                    (if (or (CameraModifier|IsDisabled _ascamera_modifier_stalker_knife_mantis) (CameraModifier|IsPendingDisable _ascamera_modifier_stalker_knife_mantis))
                      _))
                  (:Is%20Not%20Valid
                    _))
                (bind _returnvalue_2 (Game|Player|AddNewCameraModifier _playercameramanager _mantiscameramodifierclass))
                (bind _ascamera_modifier_stalker_knife_mantis_2 (Utilities|Casting|CastToCameraModifier_StalkerKnife_Mantis _returnvalue_2))
                (Variables|Default|SetMantisCameraModifier _ascamera_modifier_stalker_knife_mantis_2))
              (:else))
            (else
              (bind _playercameramanager_2 (Variables|/Script/Engine.PlayerController|GetPlayerCameraManager (Fortnite|GetFortPlayerControllerfromActor _player_pawn))
                (:then
                  (bind _returnvalue_3 (Game|Player|FindCameraModifierbyClass _playercameramanager_2 _mantiscameramodifierclass))
                  (bind _ascamera_modifier_stalker_knife_mantis_3 (Utilities|Casting|CastToCameraModifier_StalkerKnife_Mantis _returnvalue_3))
                  (Class|CameraModifierStalkerKnifeMantis|DisableandRemove _ascamera_modifier_stalker_knife_mantis_3))
                (:else)))))))
    (:else)))
